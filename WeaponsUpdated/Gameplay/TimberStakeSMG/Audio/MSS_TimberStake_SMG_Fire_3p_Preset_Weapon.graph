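# MetaSound Source preset for TimberStake SMG third-person gunfire, inheriting its graph from MSS_FPS_Weapon_Fire_3p.
# It supplies PaprikaSMG first-person shots as the close layer and shared SMG tails as medium and distant layers, with medium gain zeroed.
# Shared SMG silencer sets cover suppressed fire, alongside settings for distance curves, EQ, elevation filtering and friendly-fire volume.
# Stereo output uses shared FPS third-person attenuation and local and global gunfire concurrency sets.
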MetaSoundSource /WeaponsUpdated/Gameplay/TimberStakeSMG/Audio/MSS_TimberStake_SMG_Fire_3p_Preset_Weapon
preset_parent=/FortMetaSoundWeapon/Items/Metasounds/FPS/MSS_FPS_Weapon_Fire_3p
inputs [
 BurstRate {type=Time}
 bUseSingleVoice {type=Bool}
 Character.Affiliation {type=Enum:FortAudioAffiliation}
 Character.Health {type=Float}
 Character.UniqueID {type=Int32}
 DistanceEQ.bEnable {type=Bool}
 DistanceEQ.FilterMaxGain {type=Float}
 DistanceEQ.FilterMinGain {type=Float}
 DistanceEQ.HighShelfCutoffFrequency {type=Float}
 DistanceEQ.LowShelfCutoffFrequency {type=Float}
 Elevation.AboveGain {type=Float}
 Elevation.BelowGain {type=Float}
 Elevation.CutoffFrequency {type=Float}
 Elevation.LowCutGain {type=Float}
 FireBurst {type=Trigger}
 FireSingleShot {type=Trigger}
 MainFire.DistanceCurves {type=WaveTableBankAsset}
 MainFire.FalloffDistance {type=Float}
 MainFire.SoundsClose {type=WaveAsset:Array}
 MainFire.SoundsDistant {type=WaveAsset:Array}
 MainFire.SoundsMedium {type=WaveAsset:Array}
 Mixing.CloseGain {type=Float}
 Mixing.DistantGain {type=Float}
 Mixing.Friendly.FriendlyFireWeaponVolume {type=Float}
 Mixing.MediumGain {type=Float}
 Mixing.Suppressor.SuppressorOutputVolume {type=Float}
 Mixing.WeaponOuputVolume {type=Float}
 Random.EQ.A.MaxFreq {type=Float}
 Random.EQ.A.MaxGain {type=Float}
 Random.EQ.A.MinFreq {type=Float}
 Random.EQ.A.MinGain {type=Float}
 Random.EQ.B.MaxFreq {type=Float}
 Random.EQ.B.MaxGain {type=Float}
 Random.EQ.B.MinFreq {type=Float}
 Random.EQ.B.MinGain {type=Float}
 Random.MaxPitch {type=Float}
 Random.MaxVolume {type=Float}
 Random.MinPitch {type=Float}
 Random.MinVolume {type=Float}
 RetriggerDelayTime {type=Time}
 RoundsPerBurst {type=Int32}
 StartFiring {type=Trigger}
 StopBurst {type=Trigger}
 StopFiring {type=Trigger}
 Suppressor.BaseWeaponVolumeMultiplier {type=Float}
 Suppressor.EnvelopeAttackTime {type=Time}
 Suppressor.EnvelopeDecayTime {type=Time}
 Suppressor.Sounds.Close {type=WaveAsset:Array}
 Suppressor.Sounds.Medium {type=WaveAsset:Array}
 Suppressor.Volume.Close {type=Float}
 Suppressor.Volume.CloseBase {type=Float}
 Suppressor.Volume.Medium {type=Float}
 Suppressor.Volume.MediumBase {type=Float}
 TargetEQ.NonTargetGain {type=Float}
 UE.Attenuation.Distance {type=Float}
 UE.Source.OnPlay {type=Trigger}
 UE.Source.Orientation.Azimuth {type=Float}
 UE.Source.Orientation.Elevation {type=Float}
 UE.Spatialization.Azimuth {type=Float}
 UE.Spatialization.Elevation {type=Float}
 Weapon.AmmoRemaining {type=Int32}
 Weapon.ClipSize {type=Int32}
 Weapon.FireRate {type=Float}
 Weapon.FireStartTime {type=Float}
 Weapon.IsAimingDownSights {type=Bool}
 Weapon.IsSuppressorEnabled {type=Bool}
]
interfaces [
 Fort.Character:1.0
 Fort.Weapon:1.0
 UE.Attenuation:1.0
 UE.OutputFormat.Stereo:1.0
 UE.Source.OneShot:1.0
 UE.Source.Orientation:1.0
 UE.Source:1.1
 UE.Spatialization:1.0
]
outputs [
 bOnAlmostFinished {type=Bool}
 CloseCurveEvaluation {type=Float}
 MediumCurveEvaluation {type=Float}
 OnGunshotOut {type=Trigger}
 "UE.OutputFormat.Stereo.Audio:0" {type=Audio}
 "UE.OutputFormat.Stereo.Audio:1" {type=Audio}
 UE.Source.OneShot.OnFinished {type=Trigger}
]
pages [Default {input_defaults {
   BurstRate=0.4
   Character.Affiliation=0
   Character.Health=1
   Character.UniqueID=0
   DistanceEQ.FilterMaxGain=0.4
   DistanceEQ.FilterMinGain=-1
   DistanceEQ.HighShelfCutoffFrequency=11000
   DistanceEQ.LowShelfCutoffFrequency=150
   DistanceEQ.bEnable="<MetaSoundBuilderResult.SUCCEEDED: 0>"
   Elevation.AboveGain=0.4
   Elevation.BelowGain=-1
   Elevation.CutoffFrequency=1500
   Elevation.LowCutGain=-6
   FireBurst=null
   FireSingleShot=null
   MainFire.DistanceCurves=&WeaponRanged_DistanceCurves
   MainFire.FalloffDistance=25000
   Mixing.CloseGain=0.7
   Mixing.DistantGain=1
   Mixing.Friendly.FriendlyFireWeaponVolume=0.6
   Mixing.MediumGain=0
   Mixing.Suppressor.SuppressorOutputVolume=1
   Mixing.WeaponOuputVolume=0.85
   Random.EQ.A.MaxFreq=20000
   Random.EQ.A.MaxGain=0
   Random.EQ.A.MinFreq=699.9999
   Random.EQ.A.MinGain=-3
   Random.EQ.B.MaxFreq=8000
   Random.EQ.B.MaxGain=0
   Random.EQ.B.MinFreq=700
   Random.EQ.B.MinGain=-1
   Random.MaxPitch=1
   Random.MaxVolume=1
   Random.MinPitch=1
   Random.MinVolume=1
   RetriggerDelayTime=2
   RoundsPerBurst=2
   StartFiring=null
   StopBurst=null
   StopFiring=null
   Suppressor.BaseWeaponVolumeMultiplier=0.75
   Suppressor.EnvelopeAttackTime=0
   Suppressor.EnvelopeDecayTime=0.195
   Suppressor.Volume.Close=0.5
   Suppressor.Volume.CloseBase=1
   Suppressor.Volume.Medium=0.3
   Suppressor.Volume.MediumBase=0.3
   TargetEQ.NonTargetGain=-2
   UE.Attenuation.Distance=25000
   UE.Source.OnPlay=null
   UE.Source.Orientation.Azimuth=0
   UE.Source.Orientation.Elevation=0
   UE.Spatialization.Azimuth=0
   UE.Spatialization.Elevation=0
   Weapon.AmmoRemaining=25
   Weapon.ClipSize=25
   Weapon.FireRate=0.0833
   Weapon.FireStartTime=0
   Weapon.IsAimingDownSights=null
   Weapon.IsSuppressorEnabled=null
   bUseSingleVoice="<MetaSoundBuilderResult.SUCCEEDED: 0>"
   MainFire.SoundsClose [
    /PaprikaCoreWeapons/Items/Weapons/PaprikaSMG_DPS/Sounds/PaprikaSMG_Shoot_1P_01
    /PaprikaCoreWeapons/Items/Weapons/PaprikaSMG_DPS/Sounds/PaprikaSMG_Shoot_1P_02
    /PaprikaCoreWeapons/Items/Weapons/PaprikaSMG_DPS/Sounds/PaprikaSMG_Shoot_1P_03
    /PaprikaCoreWeapons/Items/Weapons/PaprikaSMG_DPS/Sounds/PaprikaSMG_Shoot_1P_04
    /PaprikaCoreWeapons/Items/Weapons/PaprikaSMG_DPS/Sounds/PaprikaSMG_Shoot_1P_05
    /PaprikaCoreWeapons/Items/Weapons/PaprikaSMG_DPS/Sounds/PaprikaSMG_Shoot_1P_06
    /PaprikaCoreWeapons/Items/Weapons/PaprikaSMG_DPS/Sounds/PaprikaSMG_Shoot_1P_07
    /PaprikaCoreWeapons/Items/Weapons/PaprikaSMG_DPS/Sounds/PaprikaSMG_Shoot_1P_08
   ]
   MainFire.SoundsDistant [
    /Game/Sounds/Fort_Weapons/Tails/Tails_SMGs_Distant_A_08
    /Game/Sounds/Fort_Weapons/Tails/Tails_SMGs_Distant_A_07
    /Game/Sounds/Fort_Weapons/Tails/Tails_SMGs_Distant_A_06
    /Game/Sounds/Fort_Weapons/Tails/Tails_SMGs_Distant_A_05
    /Game/Sounds/Fort_Weapons/Tails/Tails_SMGs_Distant_A_04
    /Game/Sounds/Fort_Weapons/Tails/Tails_SMGs_Distant_A_03
    /Game/Sounds/Fort_Weapons/Tails/Tails_SMGs_Distant_A_02
    /Game/Sounds/Fort_Weapons/Tails/Tails_SMGs_Distant_A_01
   ]
   MainFire.SoundsMedium [
    /Game/Sounds/Fort_Weapons/Tails/Tails_SMGs_Medium_A_01
    /Game/Sounds/Fort_Weapons/Tails/Tails_SMGs_Medium_A_02
    /Game/Sounds/Fort_Weapons/Tails/Tails_SMGs_Medium_A_03
    /Game/Sounds/Fort_Weapons/Tails/Tails_SMGs_Medium_A_04
    /Game/Sounds/Fort_Weapons/Tails/Tails_SMGs_Medium_A_05
    /Game/Sounds/Fort_Weapons/Tails/Tails_SMGs_Medium_A_06
    /Game/Sounds/Fort_Weapons/Tails/Tails_SMGs_Medium_A_07
    /Game/Sounds/Fort_Weapons/Tails/Tails_SMGs_Medium_A_08
   ]
   Suppressor.Sounds.Close [
    /Game/Sounds/Fort_Weapons/Suppressor/Silencer_SMG_Close_07
    /Game/Sounds/Fort_Weapons/Suppressor/Silencer_SMG_Close_06
    /Game/Sounds/Fort_Weapons/Suppressor/Silencer_SMG_Close_05
    /Game/Sounds/Fort_Weapons/Suppressor/Silencer_SMG_Close_04
    /Game/Sounds/Fort_Weapons/Suppressor/Silencer_SMG_Close_03
    /Game/Sounds/Fort_Weapons/Suppressor/Silencer_SMG_Close_02
    /Game/Sounds/Fort_Weapons/Suppressor/Silencer_SMG_Close_01
   ]
   Suppressor.Sounds.Medium [
    /Game/Sounds/Fort_Weapons/Suppressor/Silencer_SMG_Medium_01
    /Game/Sounds/Fort_Weapons/Suppressor/Silencer_SMG_Medium_02
    /Game/Sounds/Fort_Weapons/Suppressor/Silencer_SMG_Medium_03
    /Game/Sounds/Fort_Weapons/Suppressor/Silencer_SMG_Medium_04
    /Game/Sounds/Fort_Weapons/Suppressor/Silencer_SMG_Medium_05
    /Game/Sounds/Fort_Weapons/Suppressor/Silencer_SMG_Medium_06
    /Game/Sounds/Fort_Weapons/Suppressor/Silencer_SMG_Medium_07
   ]
  }}]
properties {
 assetImportData=&self:AssetImportData
 attenuationSettings=/FortMetaSoundWeapon/Items/Shared/Settings/FPS/AS_Weapon_FP_Fire_3P
 maxDistance=26200
 numChannels=2
 outputFormat=Stereo
 qualitySetting=Fort_Weapons
 sampleRate=48000
 soundClassObject=&Fort_WeaponRanged_Fire_3P_PSM
 concurrencyOverrides {bLimitToOwner=true, maxCount=2, resolutionRule=StopOldest}
 concurrencySet [
  /FortMetaSoundWeapon/Items/Shared/Settings/FPS/CS_Weapon_FP_Gunfire_Local
  /FortMetaSoundWeapon/Items/Shared/Settings/FPS/CS_Weapon_FP_Gunfire_Global
 ]
}
