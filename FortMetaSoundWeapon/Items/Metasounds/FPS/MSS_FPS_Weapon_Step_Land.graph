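# One-shot stereo MetaSound source for first-person weapon step/land foley, using metal-armor weapon-swap waves with local-only attenuation.
# On play, it compares PlayerRelation to zero; a true result triggers MSS_PlayRandomSound_Mono with random wave, pitch and volume inputs.
# The mono result is scaled by OutputVolume and stereo-panned by PanAmount.
# OnFinished fires when playback ends or immediately when the comparison is false.
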MetaSoundSource /FortMetaSoundWeapon/Items/Metasounds/FPS/MSS_FPS_Weapon_Step_Land
inputs [
 OutputVolume {type=Float}
 PanAmount {type=Float}
 PlayerRelation {type=Int32}
 Random.PitchShift.Max {type=Float}
 Random.PitchShift.Min {type=Float}
 Random.Volume.Max {type=Float}
 Random.Volume.Min {type=Float}
 Sounds {type=WaveAsset:Array}
 UE.Source.OnPlay {type=Trigger}
]
interfaces [UE.OutputFormat.Stereo:1.0, UE.Source.OneShot:1.0, UE.Source:1.1]
outputs [
 "UE.OutputFormat.Stereo.Audio:0" {type=Audio}
 "UE.OutputFormat.Stereo.Audio:1" {type=Audio}
 UE.Source.OneShot.OnFinished {type=Trigger}
]
pages [Default {
  input_defaults {
   OutputVolume=0.25
   PanAmount=null
   PlayerRelation=null
   Random.PitchShift.Max=1.05
   Random.PitchShift.Min=0.95
   Random.Volume.Max=null
   Random.Volume.Min=null
   UE.Source.OnPlay=null
   Sounds [
    /Game/Sounds/Fort_Player/Clothing/MetalArmor/Waves/Foley_Clothing_MetalArmor_SwapWeapon_01
    /Game/Sounds/Fort_Player/Clothing/MetalArmor/Waves/Foley_Clothing_MetalArmor_SwapWeapon_02
    /Game/Sounds/Fort_Player/Clothing/MetalArmor/Waves/Foley_Clothing_MetalArmor_SwapWeapon_03
    /Game/Sounds/Fort_Player/Clothing/MetalArmor/Waves/Foley_Clothing_MetalArmor_SwapWeapon_04
    /Game/Sounds/Fort_Player/Clothing/MetalArmor/Waves/Foley_Clothing_MetalArmor_SwapWeapon_05
   ]
  }
  nodes [
   InitVariable.Float()
   InitVariable.Float()
   InitVariable.Float()
   n1 = TriggerCompare.Clamp:Int32(A=^PlayerRelation, Compare=^UE.Source.OnPlay) {also {defaults {B=0}}}
   n2 = &MSS_PlayRandomSound_Mono(
    MaxPitch=^Random.PitchShift.Max
    MinPitch=^Random.PitchShift.Min
    MaxVolume=^Random.Volume.Max
    MinVolume=^Random.Volume.Min
    Sounds=^Sounds
    UE.Source.OnPlay=@n1.True
   )
   StereoPanner_0 = UE.Stereo Panner(
    Pan Amount=^PanAmount
    In=UE.Multiply:Audio by Float(AdditionalOperands=^OutputVolume, PrimaryOperand=@n2."UE.OutputFormat.Mono.Audio:0").Out
   )
   TriggerAnyTriggerAny2_0 = "TriggerAny.Trigger Any (2)"(In 0=@n2.UE.Source.OneShot.OnFinished, In 1=@n1.False)
  ]
  links [
   {dst="@UE.OutputFormat.Stereo.Audio:0", src=StereoPanner_0.Out Left}
   {dst="@UE.OutputFormat.Stereo.Audio:1", src=StereoPanner_0.Out Right}
   {dst="@UE.Source.OneShot.OnFinished", src=TriggerAnyTriggerAny2_0.Out}
  ]
 }]
properties {
 assetImportData=&self:AssetImportData
 attenuationSettings=/FortMetaSoundWeapon/Items/Shared/Settings/FPS/AS_Weapon_FP_Foley_LocalOnly
 maxDistance=500
 numChannels=2
 outputFormat=Stereo
 sampleRate=48000
 soundClassObject=&Fort_WeaponFoley
 concurrencySet [&CS_Fort_WeaponFoley]
}
