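# Stereo one-shot MetaSound preset of MSS_FPS_Weapon_Step_Land that supplies weapon foley for landing with the TimberStake SMG.
# The parent graph receives two borrowed rifle landing shake waves, a reduced output volume, centered pan, and random volume and raised pitch-shift ranges.
# Playback routes through the weapon foley sound class and concurrency set, with local-only first-person attenuation and a boosted source volume.
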MetaSoundSource /WeaponsUpdated/Gameplay/TimberStakeSMG/Audio/MSS_TimberStake_SMG_Land_Foley_Weapon
preset_parent=/FortMetaSoundWeapon/Items/Metasounds/FPS/MSS_FPS_Weapon_Step_Land
inputs [
 OutputVolume {type=Float}
 PanAmount {type=Float}
 PlayerRelation {type=Int32}
 Random.PitchShift.Max {type=Float}
 Random.PitchShift.Min {type=Float}
 Random.Volume.Max {type=Float}
 Random.Volume.Min {type=Float}
 Sounds {type=WaveAsset:Array}
 UE.Source.OnPlay {type=Trigger}
]
interfaces [UE.OutputFormat.Stereo:1.0, UE.Source.OneShot:1.0, UE.Source:1.1]
outputs [
 "UE.OutputFormat.Stereo.Audio:0" {type=Audio}
 "UE.OutputFormat.Stereo.Audio:1" {type=Audio}
 UE.Source.OneShot.OnFinished {type=Trigger}
]
pages [Default {input_defaults {
   OutputVolume=0.25
   PanAmount=0
   PlayerRelation=0
   Random.PitchShift.Max=1.3
   Random.PitchShift.Min=1.25
   Random.Volume.Max=1.05
   Random.Volume.Min=0.5
   UE.Source.OnPlay=null
   Sounds [
    /Game/Sounds/Fort_Weapons/MovementFoley/Rifle/WPN_Rifle_Mvmnt_Land_Shake_03
    /Game/Sounds/Fort_Weapons/MovementFoley/Rifle/WPN_Rifle_Mvmnt_Land_Shake_02
   ]
  }}]
properties {
 assetImportData=&self:AssetImportData
 attenuationSettings=/FortMetaSoundWeapon/Items/Shared/Settings/FPS/AS_Weapon_FP_Foley_LocalOnly
 maxDistance=500
 numChannels=2
 outputFormat=Stereo
 sampleRate=48000
 soundClassObject=&Fort_WeaponFoley
 volume=5
 concurrencySet [&CS_Fort_WeaponFoley]
}
